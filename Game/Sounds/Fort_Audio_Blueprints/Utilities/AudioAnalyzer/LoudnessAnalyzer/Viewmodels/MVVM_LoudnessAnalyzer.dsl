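; MVVM viewmodel for the Loudness Analyzer audio utility, exposing loudness state and button events to the UI.
; Broadcast functions fire event dispatchers for start, stop, pause, and test-tone play and stop clicks.
; InRange checks the loudness result against min and max bounds and stores the outcome.
; GetLoudnessSlateColor turns that stored flag into a green or red Slate color.
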(fn BroadcastPauseClicked
  (Default|CallOnPauseClicked))

(fn BroadcastPlayTestToneClicked
  (Default|CallOnPlayTestToneClicked))

(fn BroadcastStartClicked
  (Default|CallOnStartClicked))

(fn BroadcastStopClicked
  (Default|CallOnStopClicked))

(fn BroadcastStopTestToneClicked
  (Default|CallOnStopTestToneClicked))

(fn GetLoudnessSlateColor
  (return (Utilities|Struct|MakeSlateColor (Math|Color|SelectColor "(R=0.000000,G=0.286458,B=0.023036,A=1.000000)" "(R=1.000000,G=0.007368,B=0.000000,A=1.000000)" (Variables|Default|GetInRange)))))

(fn InRange(Loudness)
  (bind _loudnessresults_loudness (Utilities|Struct|BreakLoudnessResults (Variables|GetLoudnessResults_Loudness)))
  (bind _output_get (Variables|Default|SetInRange (Math|Float|InRange(Float) _loudnessresults_loudness (Variables|Default|GetLoudnessMin) (Variables|Default|GetLoudnessMax))))
  (return _output_get))
