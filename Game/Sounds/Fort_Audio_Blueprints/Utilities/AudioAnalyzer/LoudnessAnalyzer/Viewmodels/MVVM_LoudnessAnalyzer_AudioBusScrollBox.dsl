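; Viewmodel for an audio bus scroll box in the Loudness Analyzer audio utility, holding a list of per-item viewmodels.
; AddItem appends an item viewmodel to the Viewmodels array, RemoveItemAtIndex removes one by index, and SetItems replaces the whole array.
; UpdateItems returns the current array.
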(fn AddItem
  (params (NewItem :type MVVM%20Loudness%20Analyzer%20Audio%20Bus%20Scroll%20Box%20Item%20Object%20Reference%20%28by%20ref%29))
  (Utilities|Array|Add (Variables|Default|GetViewmodels) NewItem))

(fn RemoveItemAtIndex
  (params (Index :type Integer))
  (Utilities|Array|RemoveIndex (Variables|Default|GetViewmodels) Index))

(fn SetItems
  (params (InViewmodels :type Array%20of%20MVVM%20Loudness%20Analyzer%20Audio%20Bus%20Scroll%20Box%20Item%20Object%20References))
  (Variables|Default|SetViewmodels InViewmodels))

(fn UpdateItems
  (return (Variables|Default|GetViewmodels)))
