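; MVVM viewmodel for an audio bus scroll box item in the loudness analyzer utility, holding the latest loudness results of one analyzer.
; SetLoudnessAnalyzer stores the given LoudnessAnalyzer and, if it is valid, binds to its OnLatestOverallLoudnessResults event.
; Each event writes the incoming overall results into the LoudnessResults variable.
; The file shows no unbinding of a previously assigned analyzer.
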(event Custom|BindLoudnessAnalyzerEvents
  (AudioAnalyzer|BindEventtoOnLatestOverallLoudnessResults (Variables|Default|GetLoudnessAnalyzer) (EventDispatchers|CreateEvent :!event Custom|OnLatestOverallLoudnessResults)))

(event Custom|OnLatestOverallLoudnessResults
  (params (LatestOverallLoudnessResults :type Loudness%20Results%20Structure%20%28by%20ref%29))
  (Variables|Default|SetLoudnessResults LatestOverallLoudnessResults))

(fn SetLoudnessAnalyzer
  (params (LoudnessAnalyzer :type Loudness%20Analyzer%20Object%20Reference))
  (bind _output_get (Variables|Default|SetLoudnessAnalyzer LoudnessAnalyzer))
  (Utilities|IsValid _output_get
    (:Is%20Valid
      (CallFunction|BindLoudnessAnalyzerEvents))))
